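# Translucent lit sprite material for Niagara and Cascade particles, apparently for a popcorn emote effect, drawing an animated smoke-poof flipbook.
# An 8x8 SubUV flipbook loops from a random per-particle offset at a dynamic-parameter speed.
# Its green channel, shaped by a shadow exponent, multiplies tiled cloud noise, particle color and a boost into an emissive adjustment function.
# Its red channel times particle alpha forms the opacity mask.
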Material /Game/Effects/Fort_Effects/Emotes/M_PopCorn_01
graph {
 properties {
  bUsedWithNiagaraSprites=true
  bUsedWithParticleSprites=true
  blendMode=BLEND_Translucent
  pixelDepthOffsetMode=PDOM_Legacy
  translucencyLightingMode=TLM_VolumetricPerVertexDirectional
  usedShadingModels=Default Lit
 }
 nodes [
  n1 = ScalarParameter() {defaultValue=1, parameterName=NoiseScale}
  n2 = TextureCoordinate()
  n3 = Add(
   A=TextureSample(UVs=Multiply(A=@n1, B=@n2)) {texture=/Game/Effects/Fort_Effects/Textures/T_AtmosphericCloudNoise_Vertical}.RGB
   B=ScalarParameter() {defaultValue=1, parameterName=Noise Bias}
  )
  n4 = ScalarParameter() {defaultValue=1, parameterName=Nosie Mult}
  n5 = ParticleRandom()
  n6 = Fmod(
   A=Multiply(A=Time() {period=1}, B=DynamicParameter() {defaultValue=<rgba 1 0 0 0>, paramNames [Flipbook_Speed, 0, 0, 0]}.Flipbook_Speed)
   B=Constant() {r=1}
  )
  n7 = MaterialFunctionCall!(
   Texture=TextureObject() {samplerType=SAMPLERTYPE_LinearColor, texture=/Game/Effects/Fort_Effects/Gameplay/SmokeDustExplosions/Textures/T_SmokePoof_8x8_01}
   UVs
   SubImages=Constant2Vector() {g=8, r=8}
   Frame=Multiply(A=Add(A=@n5, B=@n6)) {constB=64}
  ) {materialFunction=/Engine/Functions/Engine_MaterialFunctions03/Texturing/SubUV_Function}
  n8 = Multiply(
   A=Saturate(None=Multiply(A=@n3, B=@n4))
   B=Power(Base=ComponentMask(None=@n7.RGB) {g=true}, Exp=ScalarParameter() {defaultValue=1, parameterName=ShadowExp})
  )
  n9 = ParticleColor()
  MaterialFunctionCall(Emissive=Multiply(A=Multiply(A=@n8, B=@n9.RGB), B=ScalarParameter() {defaultValue=0.15, parameterName=EmissiveBoost})) {materialFunction=&MF_FX_EmissiveAdjustments}
  Multiply(A=ComponentMask(None=@n7.RGB) {r=true}, B=@n9.A)
 ]
}
