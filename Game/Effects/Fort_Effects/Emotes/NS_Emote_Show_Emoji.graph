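# Niagara emote effect that pops an emoji from a user-supplied icon texture, timed by system intro, mid and outro lengths.
# Emoji_Main springs along InitialVelocity, wobbles and scales with curves driven by System.Age.
# Two GPU emitters add rainbow-tinted copies during intro and around the outro, sampling icon alpha to set renderer visibility tags.
# SimpleSpriteBurst fires a camera-oriented ring of colored, velocity-aligned streaks as the outro ends.
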NiagaraSystem /Game/Effects/Fort_Effects/Emotes/NS_Emote_Show_Emoji
dependencies {
 usedDataInterfaces [
  &NiagaraDataInterfaceCurve
  /Script/Niagara.NiagaraDataInterfaceTexture
  &NiagaraDataInterfaceColorCurve
  &NiagaraDataInterfaceArrayFloat
  &NiagaraDataInterfaceCamera
 ]
 usedDynamicInputs [&Lerp_Vector, &VectorLength]
 usedModules [
  &self:SystemSpawnScript.SystemScriptSource.SystemScriptGraph.NiagaraNodeAssignment_0.SetVariables_#1
  &self:SystemSpawnScript.SystemScriptSource.SystemScriptGraph.NiagaraNodeAssignment_5.SetVariables_#2
  &self:SystemSpawnScript.SystemScriptSource.SystemScriptGraph.NiagaraNodeAssignment_6.SetVariables_#3
  &self:SystemSpawnScript.SystemScriptSource.SystemScriptGraph.NiagaraNodeAssignment_4.SetVariables_#4
  &SystemState
  &self:SystemSpawnScript.SystemScriptSource.SystemScriptGraph.NiagaraNodeAssignment_2.SetVariables_#5
 ]
}
emitters {
 Emoji_Main {
  emitterUpdateScript [
   &EmitterState()
   &SpawnBurst_Instantaneous(Use Spawn Probability=false, UseLoopCountLimit=false)
  ]
  particleSpawnScript [
   &InitializeParticle(
    Color Mode=::NewEnumerator0
    Lifetime=$Emitter.CurrentLoopDuration
    Sprite Size Mode=::NewEnumerator3
    Uniform Sprite Size=&MakeFloatFromVector(Vector=$System.IconSize, Channel=::NewEnumerator0)
    UsePositionOffset=false
    Write Material Random=true
   )
   &CameraOffset(Camera Offset Amount=5)
  ]
  particleUpdateScript [
   &ParticleState()
   &ScaleSpriteSize(
    Scale Sprite Size Mode=::NewEnumerator0
    Uniform Scale Factor=&Multiply_Float(
     A=&FloatFromCurve(
      FloatCurve=DataInterface() {propertyValues {Curve {keys [
          {arriveTangent=1.3651, interpMode=RCIM_Cubic, leaveTangent=1.3651, tangentMode=RCTM_User, time=0, value=0.3}
          {arriveTangent=0.037, interpMode=RCIM_Cubic, leaveTangent=0.037, tangentMode=RCTM_User, time=1, value=1}
         ]}}}
      CurveIndex=&RemapRange(Input Value To Remap=$System.Age, Input Min=0, Input Max=$System.LengthIntro, Output Min=0, Output Max=1)
     )
     B=&FloatFromCurve(
      FloatCurve=DataInterface() {propertyValues {Curve {keys [
          <at 0 1 RCIM_Cubic>
          {arriveTangent=0.3845, interpMode=RCIM_Cubic, leaveTangent=0.3845, tangentMode=RCTM_User, time=0.3, value=1.425}
          {arriveTangent=-4.917, interpMode=RCIM_Cubic, leaveTangent=-4.917, tangentMode=RCTM_User, time=0.5, value=0.775}
          {arriveTangent=-0.1602, interpMode=RCIM_Cubic, leaveTangent=-0.1602, tangentMode=RCTM_User, time=0.7, value=0}
         ]}}}
      CurveIndex=&RemapRange(
       Input Value To Remap=$System.Age
       Input Min=&Add_Float(A=$System.LengthIntro, B=$System.LengthMid)
       Input Max=&Add_Float(A=$System.LengthIntro, B=&Add_Float(A=$System.LengthMid, B=$System.LengthOutro))
       Output Min=0
       Output Max=1
      )
     )
    )
   )
   &SolveForcesAndVelocity(Clamp Velocity=false, Limit Acceleration=false)
   /SetVariables(Particles.Position=&LerpPosition(
     A=$Particles.Initial.Position
     B=&AddVectorToPosition(
      Position=$Particles.Initial.Position
      Vector=&Multiply_VectorByFloat(Vector=$System.InitialVelocity, Float=$System.InitialVelocityScale)
     )
     Alpha=&FloatFromCurve(
      FloatCurve=DataInterface() {propertyValues {Curve {keys [
          {arriveTangent=0.5632, interpMode=RCIM_Cubic, leaveTangent=0.5632, tangentMode=RCTM_User, time=0, value=0}
          {arriveTangent=1.6485, interpMode=RCIM_Cubic, leaveTangent=1.6485, tangentMode=RCTM_User, time=0.5, value=0.85}
          {arriveTangent=0.0434, interpMode=RCIM_Cubic, leaveTangent=0.0434, time=0.8185, value=1.0914}
          {arriveTangent=-0.0326, interpMode=RCIM_Cubic, leaveTangent=-0.0326, time=1.7781, value=0.9055}
          {arriveTangent=0.0621, interpMode=RCIM_Cubic, leaveTangent=0.0621, time=2.6598, value=1.0314}
          {arriveTangent=-0.0295, interpMode=RCIM_Cubic, leaveTangent=-0.0295, time=3.3, value=1}
          <at 4.144 0.9876 RCIM_Cubic>
         ]}}}
      CurveIndex=&RemapRange(Input Value To Remap=$System.Age, Clamp Results=false, Input Min=0, Input Max=$System.LengthIntro, Output Min=0, Output Max=1)
     )
    )) {isSetParameters=true, module=SetVariables_#6}
   &DynamicMaterialParameters(
    Index 0 Param 2=&ClampFloat(Float=&RemapRange(
      Input Value To Remap=$System.Age
      Input Min=$System.LengthIntro
      Input Max=&Add_Float(A=$System.LengthIntro, B=$System.LengthMid)
      Output Min=0
      Output Max=1
     ))
    Module.Index0Param0Write=false
    Module.Index0Param1Write=true
    Module.Index0Param2Write=false
    Module.Index0Param3Write=false
   )
   /SetVariables(Particles.SpriteRotation:NiagaraFloat=&FloatFromCurve(
     FloatCurve=DataInterface() {propertyValues {Curve {keys [
         {arriveTangent=-0.0049, interpMode=RCIM_Cubic, leaveTangent=-0.0049, tangentMode=RCTM_User, time=0, value=0}
         {arriveTangent=-0.0376, interpMode=RCIM_Cubic, leaveTangent=-0.0376, time=1.1716, value=0.1269}
         {arriveTangent=-0.0188, interpMode=RCIM_Cubic, leaveTangent=-0.0188, time=2.3458, value=-0.0883}
         {arriveTangent=0.0246, interpMode=RCIM_Cubic, leaveTangent=0.0246, time=3.4716, value=0.0837}
         <at 4.3342 -0.0394 RCIM_Cubic>
        ]}}}
     CurveIndex=&RemapRange(Input Value To Remap=$System.Age, Clamp Results=false, Input Min=0, Input Max=$System.LengthIntro, Output Min=0, Output Max=1)
     Scale Curve=-20
    )) {isSetParameters=true}
  ]
  rendererClasses [&NiagaraSpriteRendererProperties, &NiagaraMeshRendererProperties]
  renderers [
   {rendererClass=&NiagaraSpriteRendererProperties}
   {rendererClass=&NiagaraMeshRendererProperties, rendererIndex=1}
  ]
 }
 Emoji_OffsetColor_Intro001 {
  simTarget=GPUComputeSim
  emitterSpawnScript [
   /SetVariables(Emitter.FallbackTexture=?) {isSetParameters=true}
   /SetVariables(Emitter.SubUV_Single=1) {isSetParameters=true}
  ]
  emitterUpdateScript [
   &EmitterState()
   &SpawnBurst_Instantaneous(Use Spawn Probability=false, UseLoopCountLimit=false)
   &SpawnRate(
    SpawnRate=&MakeCustomFloatFromBool(
     Bool=&SetBoolByFloatComparison(A=$System.Age, B=$System.LengthIntro, Comparison Type=::NewEnumerator6)
     True Float=10
     False Float=0
    )
    Use Spawn Probability=false
   ) {module=SpawnRate001}
  ]
  particleSpawnScript [
   /SetVariables(Particles.BirthTime:NiagaraFloat=$System.Age) {isSetParameters=true, module=SetVariables_#7}
   &InitializeParticle(
    Lifetime=&Add_Float(A=$System.LengthIntro, B=&Add_Float(A=$System.LengthMid, B=0))
    Mesh Scale Mode=::NewEnumerator3
    Mesh Uniform Scale=5
    Sprite Size Mode=::NewEnumerator3
    Uniform Sprite Size=&RandomRangeFloat(
     Minimum=&Multiply_Float(A=&MakeFloatFromVector(Vector=$System.IconSize, Channel=::NewEnumerator0), B=0.6)
     Maximum=&MakeFloatFromVector(Vector=$System.IconSize, Channel=::NewEnumerator0)
    )
    UsePositionOffset=false
    Write Material Random=true
   )
   &self:TextureSample_Alpha_01(
    SubUV_Dimensions:NiagaraFloat=$User.SubUV Columns
    Texture Sample=DataInterface() {propertyValues {Texture=null, TextureUserParameter {parameter {name=User.IconTexture, typeDefHandle {registeredTypeIndex=2}, varData []}}}}
   ) {module=TextureSample_Alpha_01}
   &DynamicMaterialParameters(Module.Index0Param0Write=false, Module.Index0Param1Write=true, Module.Index0Param2Write=false, Module.Index0Param3Write=false)
  ]
  particleUpdateScript [
   &ParticleState()
   /SetVariables(
    Particles.Position Intro:NiagaraFloat=&ClampFloat(Float=&FloatFromCurve(
      FloatCurve=DataInterface() {propertyValues {Curve {keys [
          {arriveTangent=-0.064, interpMode=RCIM_Cubic, leaveTangent=-0.064, tangentMode=RCTM_User, time=0, value=0}
          {arriveTangent=2.2008, interpMode=RCIM_Cubic, leaveTangent=2.2008, tangentMode=RCTM_User, time=0.357, value=0.4}
          {arriveTangent=0.0381, interpMode=RCIM_Cubic, leaveTangent=0.0381, tangentMode=RCTM_User, time=1, value=1}
         ]}}}
      CurveIndex=&RemapRange(Input Value To Remap=$System.Age, Input Min=$Particles.BirthTime, Input Max=$System.LengthIntro, Output Min=0, Output Max=1)
     ))
    Particles.Scale By Lifetime:NiagaraFloat=&Multiply_Float(
     A=&ClampFloat(Float=&FloatFromCurve(
       FloatCurve=DataInterface() {propertyValues {Curve {keys [
           {arriveTangent=0.2445, interpMode=RCIM_Cubic, leaveTangent=0.2445, tangentMode=RCTM_User, time=0, value=0.4}
           {arriveTangent=1.0638, interpMode=RCIM_Cubic, leaveTangent=1.0638, tangentMode=RCTM_User, time=0.5, value=0.825}
           {arriveTangent=0.037, interpMode=RCIM_Cubic, leaveTangent=0.037, tangentMode=RCTM_User, time=1, value=1}
          ]}}}
       CurveIndex=&RemapRange(Input Value To Remap=$System.Age, Input Min=$Particles.BirthTime, Input Max=$System.LengthIntro, Output Min=0, Output Max=1)
      ))
     B=&FloatFromCurve(
      FloatCurve=DataInterface() {propertyValues {Curve {keys [
          <at 0 1 RCIM_Cubic>
          {arriveTangent=1.1336, interpMode=RCIM_Cubic, leaveTangent=1.1336, tangentMode=RCTM_User, time=0.3232, value=1.475}
          {arriveTangent=-4.917, interpMode=RCIM_Cubic, leaveTangent=-4.917, tangentMode=RCTM_User, time=0.773, value=0.75}
          {arriveTangent=-0.1602, interpMode=RCIM_Cubic, leaveTangent=-0.1602, tangentMode=RCTM_User, time=1, value=0}
         ]}}}
      CurveIndex=&RemapRange(
       Input Value To Remap=$System.Age
       Input Min=&Add_Float(A=$System.LengthIntro, B=$System.LengthMid)
       Input Max=&Add_Float(A=$System.LengthIntro, B=&Add_Float(A=$System.LengthMid, B=$System.LengthOutro))
       Output Min=0
       Output Max=1
      )
     )
    )
   ) {isSetParameters=true}
   &ScaleSpriteSize(Scale Sprite Size Mode=::NewEnumerator0, Uniform Scale Factor=$Particles.Scale By Lifetime)
   /SetVariables(Particles.Position=&LerpPosition(
     A=$Particles.Initial.Position
     B=&AddVectorToPosition(
      Position=$Particles.Initial.Position
      Vector=&Multiply_VectorByFloat(Vector=$System.InitialVelocity, Float=$System.InitialVelocityScale)
     )
     Alpha=$Particles.Position Intro
    )) {isSetParameters=true}
   &Color(Color=&ColorFromCurve(
     ColorCurve=DataInterface() {propertyValues {
       AlphaCurve {keys [<at 0 1 RCIM_Cubic>]}
       BlueCurve {keys [<at 0 1 RCIM_Cubic>, <at 0.2119 1>, <at 0.5077 0.0664>, {interpMode=RCIM_Cubic, leaveTangent=-0.0452, time=0.9896, value=0.0777}]}
       GreenCurve {keys [<at 0 0 RCIM_Cubic>, <at 0.2119 0.0576>, <at 0.5077 0.915>, {interpMode=RCIM_Cubic, leaveTangent=1.4284, time=0.9896, value=1}]}
       RedCurve {keys [<at 0 0.0112 RCIM_Cubic>, <at 0.2119 0.8971>, <at 0.5077 1>, {interpMode=RCIM_Cubic, leaveTangent=-1.3447, time=0.9896, value=0.2554}]}
      }}
     CurveIndex=&RemapRange(Input Value To Remap=$Particles.BirthTime, Input Min=0, Input Max=$System.LengthIntro, Output Min=0, Output Max=1)
    ))
   /SetVariables(Particles.OutColor:LinearColor=&MultiplyLinearColor(
     Linear Color=$Particles.Color
     Scale Factor=&MakeVector4_FromFloat(Float=&FloatFromCurve(
       FloatCurve=DataInterface() {propertyValues {Curve {keys [
           {arriveTangent=2.7792, interpMode=RCIM_Cubic, leaveTangent=2.7792, tangentMode=RCTM_User, time=0, value=0}
           {arriveTangent=-2.1756, interpMode=RCIM_Cubic, leaveTangent=-2.1756, tangentMode=RCTM_User, time=0.7, value=1}
           {interpMode=RCIM_Cubic, tangentMode=RCTM_User, time=1, value=0}
          ]}}}
       CurveIndex=&ClampFloat(Float=&RemapRange(Input Value To Remap=$Emitter.Age, Input Min=0, Input Max=$System.LengthIntro, Output Min=0, Output Max=1))
       Scale Curve=&SelectFloatFromArray(
        Direct Array Index=&MakeIntFromFloat(Float=$Particles.TextureAlphaQuery)
        Float Selection Array=DataInterface() {propertyValues {FloatData [0.6,0.8]}}
       )
      ))
    )) {isSetParameters=true}
   /SetVariables(Particles.VisibilityTag=&MakeCustomIntFromBool(Bool=&SetBoolByFloatComparison(A=$Particles.TextureAlphaQuery, B=0), True INT=0, False INT=1)) {isSetParameters=true}
   /SetVariables(Particles.OutColor:LinearColor=&MultiplyLinearColor(
     Linear Color=$Particles.OutColor
     Scale Factor=&MakeVector4_FromFloat(Float=&RemapRange(Input Value To Remap=$Particles.Position Intro, Input Min=0, Input Max=1, Output Min=0.25, Output Max=0.8))
    )) {isSetParameters=true}
  ]
  rendererClasses [&NiagaraSpriteRendererProperties]
  renderers [
   {rendererClass=&NiagaraSpriteRendererProperties}
   {rendererClass=&NiagaraSpriteRendererProperties, rendererIndex=1}
  ]
 }
 Emoji_OffsetColor_Outro_Fallback {
  simTarget=GPUComputeSim
  emitterSpawnScript [
   /SetVariables(Emitter.OverlapFactor=1.6) {isSetParameters=true}
   /SetVariables(Emitter.FallbackTexture=?) {isSetParameters=true}
   /SetVariables(Emitter.SubUV_Single=1) {isSetParameters=true}
  ]
  emitterUpdateScript [
   &EmitterState()
   &SpawnRate(
    SpawnGroup=1
    SpawnRate=&MakeCustomFloatFromBool(
     Bool=&Boolean_AndOperation(
      Bool1=&SetBoolByFloatComparison(
       A=$System.Age
       B=&Add_Float(A=&Add_Float(A=$System.LengthIntro, B=$System.LengthMid), B=-0.25)
       Comparison Type=::NewEnumerator2
      )
      Bool2=&SetBoolByFloatComparison(
       A=$System.Age
       B=&Add_Float(A=&Add_Float(A=$System.LengthIntro, B=$System.LengthMid), B=0.2)
       Comparison Type=::NewEnumerator7
      )
     )
     True Float=3
     False Float=0
    )
    Use Spawn Probability=false
   ) {module=SpawnRate001}
  ]
  particleSpawnScript [
   /SetVariables(Particles.BirthTime:NiagaraFloat=$System.Age) {isSetParameters=true, module=SetVariables_#7}
   &InitializeParticle(
    Lifetime=&Multiply_Float(A=$System.LengthOutro, B=$Emitter.OverlapFactor)
    Mesh Scale Mode=::NewEnumerator3
    Mesh Uniform Scale=2
    Sprite Size Mode=::NewEnumerator3
    Uniform Sprite Size=&MakeFloatFromVector(Vector=$System.IconSize, Channel=::NewEnumerator0)
    UsePositionOffset=false
    Write Material Random=true
   )
   &self:TextureSample_Alpha_01(
    SubUV_Dimensions:NiagaraFloat=$User.SubUV Columns
    Texture Sample=DataInterface() {propertyValues {Texture=null, TextureUserParameter {parameter {name=User.IconTexture, typeDefHandle {registeredTypeIndex=2}, varData []}}}}
   ) {module=TextureSample_Alpha_01}
  ]
  particleUpdateScript [
   &ParticleState()
   /SetVariables(Particles.Scale By Lifetime:NiagaraFloat=&Add_Float(
     A=0
     B=&MaskFloatBySpawnGroup(
      Masked Float=&FloatFromCurve(
       FloatCurve=DataInterface() {propertyValues {Curve {keys [
           {arriveTangent=0.0469, interpMode=RCIM_Cubic, leaveTangent=0.0469, tangentMode=RCTM_User, time=0.3, value=1}
           {arriveTangent=1.116, interpMode=RCIM_Cubic, leaveTangent=1.116, tangentMode=RCTM_User, time=0.6, value=1.425}
           {arriveTangent=0.037, interpMode=RCIM_Cubic, leaveTangent=0.037, tangentMode=RCTM_User, time=1, value=1.6}
          ]}}}
       CurveIndex=&ClampFloat(Float=&RemapRange(
         Input Value To Remap=$System.Age
         Input Min=$Particles.BirthTime
         Input Max=&Add_Float(
          A=&Multiply_Float(A=$System.LengthOutro, B=$Emitter.OverlapFactor)
          B=&Add_Float(
           A=$System.LengthMid
           B=&Add_Float(A=&Multiply_Float(A=$System.LengthOutro, B=$Emitter.OverlapFactor), B=0.5)
          )
         )
         Output Min=0
         Output Max=1
        ))
       Scale Curve=&FloatFromCurve(FloatCurve=DataInterface() {propertyValues {Curve {keys [
            {interpMode=RCIM_Cubic, tangentMode=RCTM_User, time=0.2, value=1}
            {arriveTangent=1.5404, interpMode=RCIM_Cubic, leaveTangent=1.5404, tangentMode=RCTM_User, time=1, value=3}
           ]}}})
      )
      Passthrough Float=0
      Spawn Group Mask=1
     )
    )) {isSetParameters=true}
   &ScaleSpriteSize(Scale Sprite Size Mode=::NewEnumerator0, Uniform Scale Factor=$Particles.Scale By Lifetime)
   &SolveForcesAndVelocity(Clamp Velocity=false, Limit Acceleration=false) {enabled=false}
   /SetVariables(Particles.Position=&AddVectorToPosition(
     Position=$Particles.Initial.Position
     Vector=&Multiply_VectorByFloat(Vector=$System.InitialVelocity, Float=$System.InitialVelocityScale)
    )) {isSetParameters=true, module=SetVariables_#6}
   &DynamicMaterialParameters(
    Index 0 Param 2=1
    Index 0 Param 3=&FloatFromCurve(
     FloatCurve=DataInterface() {propertyValues {Curve {keys [
         {arriveTangent=0.2544, interpMode=RCIM_Cubic, leaveTangent=0.2544, tangentMode=RCTM_User, time=0.3, value=0}
         {arriveTangent=2.6924, interpMode=RCIM_Cubic, leaveTangent=2.6924, tangentMode=RCTM_User, time=0.6, value=0.425}
         {arriveTangent=1.2166, interpMode=RCIM_Cubic, leaveTangent=1.2166, tangentMode=RCTM_User, time=1, value=1}
        ]}}}
     CurveIndex=&RemapRange(
      Input Value To Remap=$System.Age
      Input Min=&Add_Float(A=$System.LengthIntro, B=$System.LengthMid)
      Input Max=&Add_Float(
       A=&Add_Float(A=$System.LengthIntro, B=$System.LengthMid)
       B=&Multiply_Float(A=$System.LengthOutro, B=$Emitter.OverlapFactor)
      )
      Output Min=0
      Output Max=1
     )
     Scale Curve=4
    )
    Module.Index0Param0Write=false
    Module.Index0Param1Write=false
    Module.Index0Param2Write=true
    Module.Index0Param3Write=false
   )
   &Color(Color=&MultiplyLinearColor(
     Linear Color=&ColorFromCurve(ColorCurve=DataInterface() {propertyValues {
        AlphaCurve {keys [
          <at 0 0 RCIM_Cubic>
          {arriveTangent=-1.5793, interpMode=RCIM_Cubic, leaveTangent=-1.5793, tangentMode=RCTM_User, time=0.0725, value=1}
          {arriveTangent=-1.4499, interpMode=RCIM_Cubic, leaveTangent=-1.4499, tangentMode=RCTM_User, time=0.5284, value=0.5}
          <at 1 0>
         ]}
        BlueCurve {keys [
          {arriveTangent=0.6945, interpMode=RCIM_Cubic, time=0.0213, value=1}
          <at 0.1255 1>
          <at 0.2399 0.6695>
          <at 0.3782 0.062>
          <at 0.5062 0.9229>
          <at 1 1 RCIM_Cubic>
         ]}
        GreenCurve {keys [
          {arriveTangent=-9.0114, interpMode=RCIM_Cubic, time=0.0213, value=0}
          <at 0.1255 0>
          <at 0.2399 0>
          <at 0.3782 0.5307>
          <at 0.5062 1>
          <at 1 0.9203 RCIM_Cubic>
         ]}
        RedCurve {keys [
          {arriveTangent=4.5057, interpMode=RCIM_Cubic, time=0.0213, value=0.0112}
          <at 0.1255 0.5>
          <at 0.2399 1>
          <at 0.3782 1>
          <at 0.5062 0>
          <at 1 0.1832 RCIM_Cubic>
         ]}
       }})
     Scale Factor=&MakeVector4_FromFloat(Float=&FloatFromCurve(
       FloatCurve=DataInterface() {propertyValues {Curve {keys [<at 0 1 RCIM_Cubic>, {arriveTangent=0.0724, interpMode=RCIM_Cubic, leaveTangent=0.0724, tangentMode=RCTM_User, time=1, value=0}]}}}
       Scale Curve=1.4
      ))
    )) {module=Color001}
   /SetVariables(
    Particles.CameraOffset_BG=5
    Particles.Color_BG:LinearColor=&MultiplyLinearColor(Linear Color=$Particles.Color, Scale Factor=<0.2 0.2 0.2 0.5>)
   ) {isSetParameters=true}
   /SetVariables(Particles.VisibilityTag=&MakeCustomIntFromBool(Bool=&SetBoolByFloatComparison(A=$Particles.TextureAlphaQuery, B=0), True INT=0, False INT=1)) {isSetParameters=true}
  ]
  rendererClasses [&NiagaraSpriteRendererProperties]
  renderers [
   {rendererClass=&NiagaraSpriteRendererProperties}
   {rendererClass=&NiagaraSpriteRendererProperties, rendererIndex=1}
   {rendererClass=&NiagaraSpriteRendererProperties, rendererIndex=2}
   {rendererClass=&NiagaraSpriteRendererProperties, rendererIndex=3}
  ]
 }
 SimpleSpriteBurst {
  emitterSpawnScript [/SetVariables(Emitter.OverlapFactor=1.6) {isSetParameters=true}]
  emitterUpdateScript [
   &EmitterState()
   &SpawnBurst_Instantaneous(
    Spawn Count=15
    Spawn Time=&Add_Float(
     A=&Add_Float(A=&Add_Float(A=$System.LengthIntro, B=$System.LengthMid), B=$System.LengthOutro)
     B=-0.125
    )
    Use Spawn Probability=false
    UseLoopCountLimit=true
   )
   &SpawnBurst_Instantaneous(
    Spawn Count=15
    Spawn Time=&Add_Float(
     A=&Add_Float(A=&Add_Float(A=$System.LengthIntro, B=$System.LengthMid), B=$System.LengthOutro)
     B=-0.25
    )
    Use Spawn Probability=false
    UseLoopCountLimit=false
   ) {enabled=false, module=SpawnBurst_Instantaneous001}
  ]
  particleSpawnScript [
   &InitializeParticle(
    Color=&ColorFromCurve(
     ColorCurve=DataInterface() {propertyValues {
       AlphaCurve {keys [<at 0 1 RCIM_Cubic>]}
       BlueCurve {keys [
         <at 0.0277 0.85>
         <at 0.3607 0.0375>
         {arriveTangent=1.2708, interpMode=RCIM_Cubic, leaveTangent=1.2708, time=0.4285, value=1}
         <at 0.5559 0>
         <at 1 0.85>
        ]}
       GreenCurve {keys [
         <at 0.0277 0>
         <at 0.3607 0.1975>
         {arriveTangent=-0.3089, interpMode=RCIM_Cubic, leaveTangent=-0.3089, time=0.4285, value=0.6846}
         <at 0.5559 1>
         <at 1 0>
        ]}
       RedCurve {keys [<at 0.0277 1>,<at 0.3607 1>,<at 0.4285 0 RCIM_Cubic>,<at 0.5559 0.9288>,<at 1 1>]}
      }}
     CurveIndex=&ReturnNormalizedExecIndex()
    )
    Lifetime=0.3
    Mesh Scale Mode=::NewEnumerator3
    Mesh Uniform Scale=5
    Sprite Size=<3 16>
    Sprite Size Mode=::NewEnumerator4
    UsePositionOffset=false
    Write Material Random=true
   )
   &ShapeLocation(
    MaskSpawnGroup=false
    Override Seed=false
    Ring / Disc Distribution Mode:ENiagaraSphereDistributionMode=::NewEnumerator1
    Ring Radius=&RandomRangeFloat(Minimum=3, Maximum=4)
    Rotation Axis=&ReturnCameraVector(Camera Vectors=::NewEnumerator1, Coordinate Space=::Simulation)
    Rotation Mode=::NewEnumerator1
    Shape Origin=&AddVectorToPosition(
     Position=$Particles.Initial.Position
     Vector=&Multiply_VectorByFloat(Vector=$System.InitialVelocity, Float=$System.InitialVelocityScale)
    )
    Shape Primitive=::NewEnumerator4
    U Position=&ReturnNormalizedExecIndex(Normalized Index Includes One=false)
    UseOwnerScale=true
   )
   &AddVelocity(
    Velocity=&NormalizeVector(Vector To Normalize=$Particles.ShapeLocation.ShapeVector, Safe Normalize=true, Safe Fallback Threshold=0, Safe Fallback Vector=<0 0 0.001>)
    Velocity Speed Scale=&RandomRangeFloat(Minimum=66, Maximum=125)
   )
   &SpriteFacingAndAlignment(Sprite Alignment=$Particles.Velocity, WriteAlignment=true, WriteFacing=false)
  ]
  particleUpdateScript [
   &ParticleState()
   &ScaleSpriteSize(
    Uniform Curve Scale=&RandomRangeFloat(Minimum=0.5, Maximum=1)
    Uniform Curve Sprite Scale=DataInterface() {propertyValues {Curve {keys [
        {arriveTangent=0.7922, interpMode=RCIM_Cubic, leaveTangent=0.7922, tangentMode=RCTM_User, time=0, value=0.375}
        {arriveTangent=2.4016, interpMode=RCIM_Cubic, leaveTangent=2.4016, tangentMode=RCTM_User, time=0.7508, value=2.4161}
        {arriveTangent=-1.2503, interpMode=RCIM_Cubic, leaveTangent=-1.2503, tangentMode=RCTM_User, time=1, value=0}
       ]}}}
   )
   &ScaleColor(
    Scale Alpha=&FloatFromCurve(
     FloatCurve=DataInterface() {propertyValues {Curve {keys [
         {arriveTangent=23.2471, interpMode=RCIM_Cubic, leaveTangent=23.2471, tangentMode=RCTM_User, time=0, value=0.4}
         {arriveTangent=-1.7248, interpMode=RCIM_Cubic, leaveTangent=-1.7248, tangentMode=RCTM_User, time=0.0667, value=1.0005}
         {arriveTangent=-1.5684, interpMode=RCIM_Cubic, leaveTangent=-1.5684, tangentMode=RCTM_User, time=0.2649, value=0.3207}
         {arriveTangent=-0.2526, interpMode=RCIM_Cubic, leaveTangent=-0.2526, tangentMode=RCTM_User, time=1, value=0}
        ]}}}
     Scale Curve=2
    )
    ScaleA=true
    ScaleRGB=false
   )
   &Drag(Drag=&RandomRangeFloat(Minimum=1, Maximum=3), Use Linear Drag=true, Use Rotational Drag=false)
   &SolveForcesAndVelocity(Clamp Velocity=false, Limit Acceleration=false)
  ]
  rendererClasses [&NiagaraSpriteRendererProperties]
  renderers [{rendererClass=&NiagaraSpriteRendererProperties}]
 }
}
summary {
 systemName=NS_Emote_Show_Emoji
 emitters [
  {emitterName=Emoji_Main, rendererClasses [&NiagaraSpriteRendererProperties, &NiagaraMeshRendererProperties]}
  {emitterName=Emoji_OffsetColor_Outro_Fallback, simTarget=GPUComputeSim, rendererClasses [&NiagaraSpriteRendererProperties]}
  {emitterName=Emoji_OffsetColor_Intro001, simTarget=GPUComputeSim, rendererClasses [&NiagaraSpriteRendererProperties]}
  {emitterName=SimpleSpriteBurst, rendererClasses [&NiagaraSpriteRendererProperties]}
 ]
 userVariables [
  User.IconSize {type=Vector3f, description="", defaultValue {struct=Vector3f, value=<25 0 0>}}
  User.IconTexture {type=Texture, description="", defaultValue {struct=NiagaraExt_VariableValue_Object, value {object=null, objectClass=/Script/Engine.Texture}}}
  User.IniitalPosition {type=NiagaraPosition, description="", defaultValue {}}
  User.InitialVelocity {type=Vector3f, description="", defaultValue {struct=Vector3f, value=<0 0 180>}}
  User.SubUV Columns {type=NiagaraFloat, defaultValue=1, description=""}
  User.SubUV Rows {type=NiagaraFloat, defaultValue=1, description=""}
 ]
}
